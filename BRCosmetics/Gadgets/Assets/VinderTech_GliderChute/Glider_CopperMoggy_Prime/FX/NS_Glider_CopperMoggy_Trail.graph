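# Niagara system for the CopperMoggy Prime glider's propellers and trails, with separate gameplay and front-end tuning.
# Two counter-rotating propeller meshes follow sockets, crossfading to a blurred look as an interpolated system value rises.
# Eight socket-pair AnimTrail ribbons, socket-spawned tech-shape sprites and torus-spawned speed lines ramp with system timelines and age curves.
# A variant manager selects trail and shape colors, ribbon widths and per-variant trail layers.
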NiagaraSystem /BRCosmetics/Gadgets/Assets/VinderTech_GliderChute/Glider_CopperMoggy_Prime/FX/NS_Glider_CopperMoggy_Trail
dependencies {
 usedDataInterfaces [
  &NiagaraDataInterfaceCurve
  &NiagaraDataInterfaceArrayInt32
  &NiagaraDataInterfaceVector2DCurve
  &NiagaraDataInterfaceColorCurve
  &NiagaraDataInterfaceArrayColor
  &NiagaraDataInterfaceSocketReader
 ]
 usedDynamicInputs [
  &RandomRangeVector2D
  &SetBoolByFloatComparison
  &Abs_Float
  &NormalizeVector
  &Add_Vector
  &TransformVector
 ]
 usedModules [
  &self:SystemSpawnScript.SystemScriptSource.SystemScriptGraph.NiagaraNodeAssignment_1.SetVariables_#1
  &SystemState
  &Timeline
  &FN_Integrator
  &InterpolateOverTime
  &FN_VariantManager
 ]
}
emitters {
 Propeller {
  emitterUpdateScript [
   &EmitterState(Life Cycle Mode=::NewEnumerator1, Loop Behavior=::NewEnumerator1, Loop Duration Mode=::NewEnumerator1, UseLoopDelay=false)
   &SpawnBurst_Instantaneous(Spawn Count=2, Use Spawn Probability=false, UseLoopCountLimit=false)
   /SetVariables(Emitter.ScaleDelay:NiagaraFloat=&FloatFromCurve(
     FloatCurve=DataInterface() {propertyValues {Curve {keys [{tangentMode=RCTM_User, time=1, value=0}, {tangentMode=RCTM_User, time=1.1, value=1}]}}}
     CurveIndex=$Emitter.Age
    )) {isSetParameters=true}
  ]
  particleSpawnScript [
   &InitializeParticle(Color Mode=::NewEnumerator0, Lifetime=5, Mesh Scale Mode=::NewEnumerator3, UsePositionOffset=false, Write Material Random=true)
   /SetVariables(Particles.MotionBlurScale:Vector3f=<0.4 0.4 0.1>) {isSetParameters=true}
   /SetVariables(Particles.MotionBlurDynamic:Vector4f=&MakeVector4(X=-4, Y=6, Z=0, W=0)) {isSetParameters=true}
   &PartitionParticles()
   /SetVariables(Particles.RotateSpeed=5) {isSetParameters=true}
  ]
  particleUpdateScript [
   &ParticleState(Kill Particles When Lifetime Has Elapsed=false, Let Infinitely Lived Particles Die When Emitter Deactivates=true)
   &ScaleMeshSize(Scale Factor=&Multiply_VectorByFloat(Vector=<1 1 1>, Float=$Emitter.ScaleDelay))
   &SocketLocation(
    Orientation Sampling=::NewEnumerator1
    SamplePosOffset=true
    Socket Index=&SelectIntFromArray(Direct Array Index=$Particles.PartitionParticles.Partition, Int Selection Array=DataInterface() {propertyValues {IntData [8,9]}})
    Socket Reader=$System.Socket Reader
    Socket Sampling Mode=::NewEnumerator3
    VelocitySpeedLimit=false
   )
   /SetVariables(Particles.RotateSpeed:NiagaraFloat=&Add_Float(
     A=$Particles.RotateSpeed
     B=&Multiply_Float(
      A=$Engine.DeltaTime
      B=&FloatFromCurve(
       FloatCurve=DataInterface() {propertyValues {Curve {defaultValue=1, keys [<at 0 1 RCIM_Cubic>,<at 0.7 3 RCIM_Cubic>]}}}
       CurveIndex=$System.InterpolateOverTime002.InterpolateOverTime_Float.Moving Average
       Scale Curve=&Multiply_Float(
        A=700
        B=&MergeHelper_FloatSwitch(0=1, 1=-1, 2=?, 3=?, Index=$Particles.PartitionParticles.Partition, Num=::NewEnumerator0)
       )
      )
     )
    )) {isSetParameters=true}
   /SetVariables(Particles.MeshOrientation=&MakeQuaternion(
     Quaternion From=3
     Axis=&RotateVectorByQuaternion(Vector To Rotate=<1 0 0>, Quaternion=$Output.SocketLocation.SampledRotation)
     Angle=&Multiply_Float(A=$Emitter.Age, B=0.2)
     Pitch=$Emitter.Age
     From=<0 0 1>
     To=&RotateVectorByQuaternion(Vector To Rotate=<0 0 1>, Quaternion=$Output.SocketLocation.SampledRotation)
     Stabilization Mode:ENiagara_OrientVectorCount=::NewEnumerator0
     Side Direction:Vector3f=?
     Up Direction:Vector3f=?
     Angle Type=::NewEnumerator0
     Normalize:NiagaraBool=true
     Evaluation Type=::NewEnumerator1
    )) {isSetParameters=true}
   /SetVariables(Particles.MeshOrientation=&MultiplyQuaternion(
     Quaternion A=$Particles.MeshOrientation
     Quaternion B=&MakeQuaternion(
      Quaternion From=0
      Axis=<0 0 1>
      Angle=$Particles.RotateSpeed
      From=?
      To=?
      Stabilization Mode:ENiagara_OrientVectorCount=::NewEnumerator0
      Side Direction:Vector3f=?
      Up Direction:Vector3f=?
      Angle Type=::NewEnumerator0
      Normalize:NiagaraBool=true
     )
    )) {isSetParameters=true}
   /SetVariables(
    Particles.BlurPropColor:LinearColor=&MakeLinearColorFromVectorAndFloat(
     "Vector (RGB)"=&MakeVectorFromLinearColorRGB(Color=<rgba 2 2 2 1>)
     "Float (Alpha)"=&FN_FrontEndFloatModifier(
      Gameplay=&Lerp_Float(A=0, B=0.5, Alpha=$System.InterpolateOverTime002.InterpolateOverTime_Float.Moving Average)
      Front End=1
     )
    )
    Particles.MainPropColor:LinearColor=&MakeLinearColorFromVectorAndFloat(
     "Vector (RGB)"=<1 1 1>
     "Float (Alpha)"=&FN_FrontEndFloatModifier(
      Gameplay=&FloatFromCurve(
       FloatCurve=DataInterface() {propertyValues {Curve {keys [<at 0 1>,<at 0.6 0>]}}}
       CurveIndex=$System.InterpolateOverTime002.InterpolateOverTime_Float.Moving Average
      )
      Front End=0
     )
    )
   ) {isSetParameters=true}
   /SetVariables(Particles.MotionBlurScale:Vector3f=&Multiply_VectorByFloat(Vector=$Particles.Initial.MotionBlurScale, Float=$Emitter.ScaleDelay)) {isSetParameters=true}
   /SetVariables(Particles.BlurCameraOffset=2) {isSetParameters=true}
  ]
  rendererClasses [&NiagaraMeshRendererProperties]
  renderers [
   {rendererClass=&NiagaraMeshRendererProperties, rendererIndex=0}
   {rendererClass=&NiagaraMeshRendererProperties, rendererIndex=1}
  ]
 }
 Speed_Lines {
  emitterUpdateScript [
   &EmitterState()
   &SpawnRate(
    SpawnRate=&FN_FrontEndFloatModifier(Gameplay=&Lerp_Float(A=0, B=90, Alpha=$System.Timeline.NormalizedTime), Front End=50)
    Use Spawn Probability=true
   ) {module=SpawnRate001}
  ]
  particleSpawnScript [
   &InitializeParticle(
    Color=&MakeLinearColorFromVectorAndFloat(
     "Vector (RGB)"=<1 1 1>
     "Float (Alpha)"=&Multiply_Float(
      A=&UniformRangedFloat(Minimum=0.1, Maximum=0.15)
      B=&FN_FrontEndFloatModifier(Gameplay=0.5, Front End=0.5)
     )
    )
    Lifetime Max=0.16
    Lifetime Min=0.1
    Lifetime Mode=::NewEnumerator1
    Sprite Size Max=<3 10>
    Sprite Size Min=<2 2>
    Sprite Size Mode=::NewEnumerator2
    UsePositionOffset=false
    Write Material Random=true
   )
   &TorusLocation(
    Align Torus With Axis=true
    Handle Radius=0
    Large Radius:NiagaraFloat=&FN_FrontEndFloatModifier(
     Gameplay=&RandomRangeFloat(Minimum=150, Maximum=200)
     Front End=&RandomRangeFloat(Minimum=100, Maximum=200)
    )
    MaskSpawnGroup=false
    Module.Override Randomness=false
    Torus Axis=<0 1 0>
    Torus Origin:NiagaraPosition=&AddVectorToPosition(
     Position=$Particles.Position
     Vector=&FN_FrontEndVector3Modifier(Gameplay=<0 0 0>, Front End=<0 100 0>)
    )
   )
   &AddVelocityFromPoint(
    Module.Use Falloff Distance=false
    Origin Offset=<0 -600 0>
    Velocity Strength=&Multiply_Float(
     A=&FN_FrontEndFloatModifier(Gameplay=600, Front End=0)
     B=&MakeCustomFloatFromBool(Bool=$User.bIsFrontEndPreview, True Float=1, False Float=1)
    )
   )
   &AddVelocity(Velocity=&Multiply_Vector(
     A=&RandomRangeVector(Minimum=<0 300 0>, Maximum=<0 700 0>)
     B=&MakeVector(X=1, Y=&MakeCustomFloatFromBool(Bool=$User.bIsFrontEndPreview, True Float=-1, False Float=1), Z=1)
    ))
  ]
  particleUpdateScript [
   &SolveForcesAndVelocity(Clamp Velocity=false, Limit Acceleration=false)
   &ParticleState()
   &ScaleSpriteSize(Scale Factor=&Vector2DFromCurve(
     Vector2Curve=DataInterface() {propertyValues {XCurve {keys [<at 0 0.5 RCIM_Cubic>,<at 0.5 1 RCIM_Cubic>]}, YCurve {keys [<at 0 0 RCIM_Cubic>,<at 0.5 1 RCIM_Cubic>]}}}
     ScaleCurve=&Vector2DFromFloat(Value=1)
    ))
   &ScaleSpriteSizeBySpeed(Max Scale Factor=<1 50>, Min Scale Factor=<1 40>, Velocity Threshold=100)
   &ScaleColor(
    Scale Alpha=&Multiply_Float(
     A=&FN_FrontEndFloatModifier(
      Gameplay=&Lerp_Float(A=0, B=1, Alpha=$System.Timeline.NormalizedTime)
      Front End=&Lerp_Float(A=0, B=1, Alpha=$System.Timeline002.NormalizedTime)
     )
     B=&FloatFromCurve(
      FloatCurve=DataInterface() {propertyValues {Curve {
         defaultValue=1
         keys [{interpMode=RCIM_Cubic, tangentMode=RCTM_User, time=2.3, value=0}, {interpMode=RCIM_Cubic, tangentMode=RCTM_User, time=2.8, value=1}]
        }}}
      CurveIndex=$Emitter.Age
      Scale Curve=0.5
     )
    )
    Scale RGB=&VectorFromFloat(Value=1)
    ScaleA=true
    ScaleRGB=true
   )
  ]
  rendererClasses [&NiagaraSpriteRendererProperties]
  renderers [{rendererClass=&NiagaraSpriteRendererProperties}]
 }
 Tech_Shapes {
  emitterUpdateScript [
   &EmitterState(Life Cycle Mode=::NewEnumerator1, Loop Behavior=::NewEnumerator1, Loop Delay=2.4, Loop Duration Mode=::NewEnumerator1, UseLoopDelay=true)
   &SpawnRate(
    SpawnRate=&FN_FrontEndFloatModifier(
     Gameplay=&Multiply_Float(
      A=/BRCosmetics/Weapons/FORT_Melee/Pickaxe_GateHound/FX/VariantManager/VariantFloatSelector(
       Original Float:NiagaraFloat=50
       Variant 1 Float:NiagaraFloat=50
       Variant 2 Float:NiagaraFloat=50
       Variant 3 Float:NiagaraFloat=?
       Variant Amount:NiagaraInt32=1
      )
      B=$System.Timeline.Time
     )
     Front End=50
    )
    Use Spawn Probability=false
   )
  ]
  particleSpawnScript [
   &FN_SetSpawnGroup(Number of Spawn Groups=2)
   /SetVariables(Particles.VisibilityTag=&RandomRangeInt(Minimum=0, Maximum=1)) {isSetParameters=true}
   /SetVariables(Particles.VisibilityTag=&Add_Integer(
     A=$Particles.VisibilityTag
     B=&Multiply_Int(A=&MakeIntFromFloat(Float=$System.VariantIndex), B=2)
    )) {isSetParameters=true}
   &InitializeParticle(
    Color=&VariantColorSelector(
     Variant Amount=1
     Original Color=&SelectLinearColorFromArray(
      Array Sampling Mode=::NewEnumerator0
      Direct Array Index=?
      Color Selection Array=DataInterface() {propertyValues {ColorData [<rgba 1 0 0 2>, <rgba 1.5 0.05 0 1>, <rgba 1 0.0833 0 2>]}}
      Random Seed=0
     )
     Variant 1 Color=&SelectLinearColorFromArray(
      Array Sampling Mode=::NewEnumerator0
      Direct Array Index=?
      Color Selection Array=DataInterface() {propertyValues {ColorData [<rgba 0 1.5 0.5 1>, <rgba 0 1.5 0.75 1>, <rgba 0 1 0.7096 1>]}}
      Random Seed=0
     )
     Variant 2 Color=&SelectLinearColorFromArray(
      Array Sampling Mode=::NewEnumerator0
      Direct Array Index=?
      Color Selection Array=DataInterface() {propertyValues {ColorData [<rgba 1.5 1.125 0 1>, <rgba 1.5 1.05 0 1>, <rgba 1 0.8333 0 1>]}}
      Random Seed=0
     )
    )
    Lifetime=0.4
    Sprite Rotation Mode=::NewEnumerator2
    Sprite Size Mode=::NewEnumerator3
    Uniform Sprite Size=&RandomRangeFloat(Minimum=4, Maximum=8)
    UsePositionOffset=true
    Write Material Random=true
   )
   &SocketLocation(
    Position Sampling=::NewEnumerator1
    SamplePosOffset=true
    Sampled Position Offset=&FN_FrontEndVector3Modifier(Gameplay=<0 20 0>, Front End=<0 -20 0>)
    Sampled Position Offset Coordinate Space:ENiagaraSkelMeshTransforms=::NewEnumerator2
    Socket Index=&MaskIntBySpawnGroup(Masked Int=0, Passthrough Int=4, Spawn Group Mask=0)
    Socket Reader=$System.Socket Reader
    Socket Sampling Mode=::NewEnumerator3
    VelocitySpeedLimit=false
   )
   &SocketLocation(
    Position Sampling=::NewEnumerator1
    SamplePosOffset=true
    Socket Index=&MaskIntBySpawnGroup(Masked Int=3, Passthrough Int=7, Spawn Group Mask=0)
    Socket Reader=$System.Socket Reader
    Socket Sampling Mode=::NewEnumerator3
    VelocitySpeedLimit=false
   ) {module=SocketLocation001}
   &AddVelocity(
    Cone Angle=&FN_FrontEndFloatModifier(Gameplay=&Lerp_Float(A=5, B=10, Alpha=$System.Timeline.NormalizedTime), Front End=10)
    Cone Axis=<0 1 0>
    Module.UseConeFalloff=false
    Module.UseVelDistribution=false
    Override Seed=false
    Velocity Mode=::NewEnumerator2
    Velocity Speed=&FN_FrontEndFloatModifier(
     Gameplay=&Multiply_Float(
      A=&Lerp_Float(A=0, B=&RandomRangeFloat(Minimum=100, Maximum=200), Alpha=$System.Timeline.NormalizedTime)
      B=&FloatFromCurve(
       FloatCurve=DataInterface() {propertyValues {Curve {
          defaultValue=1
          keys [{interpMode=RCIM_Cubic, tangentMode=RCTM_User, time=1, value=0}, {interpMode=RCIM_Cubic, tangentMode=RCTM_User, time=2, value=1}]
         }}}
       CurveIndex=$Emitter.Age
      )
     )
     Front End=&RandomRangeFloat(Minimum=-100, Maximum=-200)
    )
   )
   &CameraOffset(Camera Offset Amount=5)
   /SetVariables(Particles.Position=&Lerp_Vector(
     A=Linked() {linkedVariable {name=Output.SocketLocation.SampledPosition, type=NiagaraPosition}}
     B=Linked() {linkedVariable {name=Output.SocketLocation001.SampledPosition, type=NiagaraPosition}}
     Alpha=&RandomRangeFloat(Minimum=0, Maximum=1)
    )) {isSetParameters=true}
  ]
  particleUpdateScript [
   &ParticleState()
   &Drag(Drag=0.5, Use Linear Drag=true, Use Rotational Drag=false)
   &ScaleSpriteSize(Scale Factor=&Vector2DFromFloat(Value=&FloatFromCurve(FloatCurve=DataInterface() {propertyValues {Curve {
         defaultValue=1
         keys [{interpMode=RCIM_Cubic, tangentMode=RCTM_User, time=0, value=0.5}, {interpMode=RCIM_Cubic, tangentMode=RCTM_User, time=0.7, value=1}]
        }}}))) {module=ScaleSpriteSize001}
   &ScaleColor(
    Scale Alpha=&FloatFromCurve(FloatCurve=DataInterface() {propertyValues {Curve {
        defaultValue=0
        keys [
         {interpMode=RCIM_Cubic, tangentMode=RCTM_User, time=0, value=0}
         {interpMode=RCIM_Cubic, tangentMode=RCTM_User, time=0.2, value=1}
         {tangentMode=RCTM_User, time=1, value=0}
        ]
       }}})
    ScaleA=true
    ScaleRGB=true
   ) {module=ScaleColor001}
   &CurlNoiseForce(
    Bias Noise=false
    Noise Strength=&Multiply_Float(A=500, B=&Lerp_Float(A=0.3, B=1, Alpha=$System.Timeline.NormalizedTime))
    Pan Noise=false
   )
   &SolveForcesAndVelocity(Clamp Velocity=false, Limit Acceleration=false)
   &SpriteFacingAndAlignment(Sprite Facing=$Particles.ShapeLocation.ShapeTangent, WriteAlignment=false, WriteFacing=true)
   &FN_Sprite_Attributes(
    Color=&MultiplyLinearColor(Linear Color=$Particles.Color, Scale Factor=<1 1 1 1>)
    Create Color Attribute=true
    Create SpriteSize Attribute=true
    Module Output:ENiagara_AttributeSamplingApplyOutput=::NewEnumerator1
    SpriteSize=&Multiply_Vector2D_ByFloat(Float=1.5, Vector2D=$Particles.SpriteSize)
   )
  ]
  rendererClasses [&NiagaraSpriteRendererProperties]
  renderers [
   {rendererClass=&NiagaraSpriteRendererProperties}
   {rendererClass=&NiagaraSpriteRendererProperties, rendererIndex=1}
   {rendererClass=&NiagaraSpriteRendererProperties, rendererIndex=2}
   {rendererClass=&NiagaraSpriteRendererProperties, rendererIndex=3}
   {rendererClass=&NiagaraSpriteRendererProperties, rendererIndex=4}
   {rendererClass=&NiagaraSpriteRendererProperties, rendererIndex=5}
  ]
 }
 Trails {
  emitterSpawnScript [&FN_RibbonIDOffsetUtility()]
  emitterUpdateScript [
   &EmitterState(Life Cycle Mode=::NewEnumerator1, Loop Behavior=::NewEnumerator1, Loop Duration Mode=::NewEnumerator1, UseLoopDelay=false)
   &FN_SpawnAnimTrail(Number of Anim Trails=8, Spawn Interval=0)
  ]
  particleSpawnScript [
   &FN_InitializeAnimTrailBoneIndecies()
   &FN_InitializeAnimTrailBoneIndecies(AnimTrail Selection=1, First Bone or Socket Index=1, Second Bone or Socket Index=2) {module=FN_InitializeAnimTrailBoneIndecies001}
   &FN_InitializeAnimTrailBoneIndecies(AnimTrail Selection=2, First Bone or Socket Index=2, Second Bone or Socket Index=3) {module=FN_InitializeAnimTrailBoneIndecies002}
   &FN_InitializeAnimTrailBoneIndecies(AnimTrail Selection=3, First Bone or Socket Index=4, Second Bone or Socket Index=5) {module=FN_InitializeAnimTrailBoneIndecies003}
   &FN_InitializeAnimTrailBoneIndecies(AnimTrail Selection=4, First Bone or Socket Index=5, Second Bone or Socket Index=6) {module=FN_InitializeAnimTrailBoneIndecies004}
   &FN_InitializeAnimTrailBoneIndecies(AnimTrail Selection=5, First Bone or Socket Index=6, Second Bone or Socket Index=7) {module=FN_InitializeAnimTrailBoneIndecies005}
   &FN_InitializeAnimTrailBoneIndecies(AnimTrail Selection=6, Second Bone or Socket Index=3) {module=FN_InitializeAnimTrailBoneIndecies006}
   &FN_InitializeAnimTrailBoneIndecies(AnimTrail Selection=7, First Bone or Socket Index=4, Second Bone or Socket Index=7) {module=FN_InitializeAnimTrailBoneIndecies007}
   &FN_InitializeAnimTrail(
    AnimTrailWidth=&VariantFloatSelector(
     Original Float:NiagaraFloat=1.05
     Variant 1 Float:NiagaraFloat=1
     Variant 2 Float:NiagaraFloat=1.05
     Variant 3 Float:NiagaraFloat=?
     Variant Amount:NiagaraInt32=1
    )
    Color=&VariantColorSelector(Variant Amount=1, Original Color=<rgba 1 0 0.0761 1>, Variant 1 Color=<rgba 0.2 2 0.95 0.5>, Variant 2 Color=<rgba 1 0.5 0 1>)
    Lifetime=&MakeCustomFloatFromBool(Bool=$Particles.Locked, True Float=0, False Float=0.22)
    Position Offset=&FN_FrontEndVector3Modifier(Gameplay=<0 -5 0>, Front End=<0 5 0>)
   )
   &AddVelocity(Velocity=&Multiply_VectorByFloat(
     Vector=<0 1 0>
     Float=&FN_FrontEndFloatModifier(
      Gameplay=&Lerp_Float(
       A=400
       B=800
       Alpha=&FloatFromCurve(
        FloatCurve=DataInterface() {propertyValues {Curve {
           defaultValue=1
           keys [{interpMode=RCIM_Cubic, tangentMode=RCTM_User, time=2.1, value=0}, {interpMode=RCIM_Cubic, tangentMode=RCTM_User, time=2.7, value=1}]
          }}}
        CurveIndex=$Emitter.Age
        Scale Curve=$System.Timeline.Time
       )
      )
      Front End=&Lerp_Float(
       A=-400
       B=-800
       Alpha=&FloatFromCurve(
        FloatCurve=DataInterface() {propertyValues {Curve {
           defaultValue=1
           keys [{interpMode=RCIM_Cubic, tangentMode=RCTM_User, time=2, value=0}, {interpMode=RCIM_Cubic, tangentMode=RCTM_User, time=2.5, value=1}]
          }}}
        CurveIndex=$Emitter.Age
       )
      )
     )
    ))
  ]
  particleUpdateScript [
   &SolveForcesAndVelocity(Clamp Velocity=false, Limit Acceleration=false)
   &FN_UpdateAnimTrail()
   &ScaleRibbonWidth(Ribbon Width Scale=&FloatFromCurve(FloatCurve=DataInterface() {propertyValues {Curve {defaultValue=0, keys [<at 0 1>]}}}))
   &DynamicMaterialParameters(
    Index 0 Param 1=&FN_FrontEndFloatModifier(Gameplay=0.25, Front End=0.25)
    Index 0 Param 2=&FN_FrontEndFloatModifier(
     Gameplay=&MergeHelper_FloatSwitch(
      0=0.5
      1=&Lerp_Float(
       A=-0.1
       B=0.3
       Alpha=&FN_FrontEndFloatModifier(Gameplay=$System.Timeline.Time, Front End=$System.Timeline002.Time)
      )
      2=0.5
      3=?
      Index=&MakeIntFromFloat(Float=$System.VariantIndex)
      Num=::NewEnumerator1
     )
     Front End=0.5
    )
    Index 0 Param 3=&FN_FrontEndFloatModifier(Gameplay=&Lerp_Float(A=0, B=0.2, Alpha=$System.Timeline.Time), Front End=0.15)
    Index 0 Param 4=&FN_FrontEndFloatModifier(Gameplay=&Lerp_Float(A=0.3, B=0.15, Alpha=$System.Timeline.Time), Front End=0.15)
    Module.Index0Float0Write=true
    Module.Index0Float1Write=true
    Module.Index0Float2Write=true
    Module.Index0Float3Write=true
   )
   /SetVariables(Particles.DynForSoftTrail:Vector4f=&MakeVector4(X=1, Y=0.75, Z=&FN_FrontEndFloatModifier(Gameplay=1, Front End=1), W=0)) {isSetParameters=true}
   /SetVariables(Particles.RibbonWidth02:NiagaraFloat=&Multiply_Float(
     A=$Particles.RibbonWidth
     B=&VariantFloatSelector(
      Original Float:NiagaraFloat=1.02
      Variant 1 Float:NiagaraFloat=1.05
      Variant 2 Float:NiagaraFloat=1
      Variant 3 Float:NiagaraFloat=?
      Variant Amount:NiagaraInt32=1
     )
    )) {isSetParameters=true}
   &ScaleColor(
    Scale Alpha=&Multiply_Float(
     A=&FloatFromCurve(
      FloatCurve=DataInterface() {propertyValues {Curve {
         defaultValue=1
         keys [{interpMode=RCIM_Cubic, tangentMode=RCTM_User, time=0, value=1}, {interpMode=RCIM_Cubic, tangentMode=RCTM_User, time=0.7, value=0}]
        }}}
      Scale Curve=&FloatFromCurve(
       FloatCurve=DataInterface() {propertyValues {Curve {
          defaultValue=1
          keys [{interpMode=RCIM_Cubic, tangentMode=RCTM_User, time=2.4, value=0}, {interpMode=RCIM_Cubic, tangentMode=RCTM_User, time=2.8, value=1}]
         }}}
       CurveIndex=$Emitter.Age
       Scale Curve=&FN_FrontEndFloatModifier(Gameplay=$System.Timeline.Time, Front End=$System.Timeline002.Time)
      )
     )
     B=&MakeCustomFloatFromBool(
      Bool=&SetBoolByIntComparison(A=&SpawnGroup(), B=6, Comparison Type=::NewEnumerator6)
      True Float=0
      False Float=0.5
     )
    )
    ScaleA=true
    ScaleRGB=true
   )
   /SetVariables(Particles.SwanColor:LinearColor=&MakeLinearColorFromVectorAndFloat(
     "Vector (RGB)"=&MakeVectorFromLinearColorRGB(Color=<rgba 1 0.1 0.25 1>)
     "Float (Alpha)"=&Multiply_Float(
      A=&FloatFromCurve(
       FloatCurve=DataInterface() {propertyValues {Curve {
          defaultValue=1
          keys [{interpMode=RCIM_Cubic, tangentMode=RCTM_User, time=2.3, value=0}, {interpMode=RCIM_Cubic, tangentMode=RCTM_User, time=2.8, value=1}]
         }}}
       CurveIndex=$Emitter.Age
       Scale Curve=&FN_FrontEndFloatModifier(Gameplay=$System.Timeline.Time, Front End=$System.Timeline002.Time)
      )
      B=&Multiply_Float(
       A=&MakeCustomFloatFromBool(
        Bool=&SetBoolByIntComparison(A=&SpawnGroup(), B=6, Comparison Type=::NewEnumerator6)
        True Float=1
        False Float=0
       )
       B=&VariantFloatSelector(
        Original Float:NiagaraFloat=1
        Variant 1 Float:NiagaraFloat=0
        Variant 2 Float:NiagaraFloat=0
        Variant 3 Float:NiagaraFloat=?
        Variant Amount:NiagaraInt32=1
       )
      )
     )
    )) {isSetParameters=true}
   /SetVariables(Particles.PrimeColor:LinearColor=&MakeLinearColorFromVectorAndFloat(
     "Vector (RGB)"=&MakeVectorFromLinearColorRGB(Color=<rgba 1 1 1 1>)
     "Float (Alpha)"=&Multiply_Float(
      A=&FloatFromCurve(
       FloatCurve=DataInterface() {propertyValues {Curve {
          defaultValue=1
          keys [{interpMode=RCIM_Cubic, tangentMode=RCTM_User, time=2.3, value=0}, {interpMode=RCIM_Cubic, tangentMode=RCTM_User, time=2.8, value=1}]
         }}}
       CurveIndex=$Emitter.Age
       Scale Curve=&FN_FrontEndFloatModifier(Gameplay=$System.Timeline.Time, Front End=$System.Timeline002.Time)
      )
      B=&Multiply_Float(
       A=&MakeCustomFloatFromBool(
        Bool=&SetBoolByIntComparison(A=&SpawnGroup(), B=6, Comparison Type=::NewEnumerator6)
        True Float=1
        False Float=0
       )
       B=&VariantFloatSelector(
        Original Float:NiagaraFloat=0
        Variant 1 Float:NiagaraFloat=0.75
        Variant 2 Float:NiagaraFloat=0
        Variant 3 Float:NiagaraFloat=?
        Variant Amount:NiagaraInt32=1
       )
      )
     )
    )) {isSetParameters=true}
   /SetVariables(Particles.FlameColor:LinearColor=&MakeLinearColorFromVectorAndFloat(
     "Vector (RGB)"=&MakeVectorFromLinearColorRGB(Color=<rgba 0.5826 0.437 0 2>)
     "Float (Alpha)"=&Multiply_Float(
      A=&FloatFromCurve(
       FloatCurve=DataInterface() {propertyValues {Curve {
          defaultValue=1
          keys [{interpMode=RCIM_Cubic, tangentMode=RCTM_User, time=2.3, value=0}, {interpMode=RCIM_Cubic, tangentMode=RCTM_User, time=2.8, value=1}]
         }}}
       CurveIndex=$Emitter.Age
       Scale Curve=&FN_FrontEndFloatModifier(Gameplay=$System.Timeline.Time, Front End=$System.Timeline002.Time)
      )
      B=&Multiply_Float(
       A=&MakeCustomFloatFromBool(
        Bool=&SetBoolByIntComparison(A=&SpawnGroup(), B=6, Comparison Type=::NewEnumerator6)
        True Float=1
        False Float=0
       )
       B=&VariantFloatSelector(
        Original Float:NiagaraFloat=0
        Variant 1 Float:NiagaraFloat=0
        Variant 2 Float:NiagaraFloat=1
        Variant 3 Float:NiagaraFloat=?
        Variant Amount:NiagaraInt32=1
       )
      )
     )
    )) {isSetParameters=true}
   /SetVariables(Particles.Binding_Position:NiagaraPosition=&Divide_Vector(
     A:Vector3f=Linked() {linkedVariable {name=Particles.Position, type=NiagaraPosition}}
     B:Vector3f=&FN_FrontEndVector3Modifier(Gameplay=<1 1 1>, Front End=$Engine.Owner.Scale)
    )) {isSetParameters=true}
   /SetVariables(Particles.PrimeTrailDynamic:Vector4f=&MakeVector4(X=&Multiply_Float(A=&OneMinusFloat(Float=$Engine.RealTime), B=0.75), Y=1, Z=1, W=1)) {isSetParameters=true}
  ]
  rendererClasses [&NiagaraRibbonRendererProperties]
  renderers [
   {rendererClass=&NiagaraRibbonRendererProperties}
   {rendererClass=&NiagaraRibbonRendererProperties, rendererIndex=1}
   {rendererClass=&NiagaraRibbonRendererProperties, rendererIndex=2}
   {rendererClass=&NiagaraRibbonRendererProperties, rendererIndex=3}
  ]
 }
}
summary {
 systemName=NS_Glider_CopperMoggy_Trail
 emitters [
  {emitterName=Trails, rendererClasses [&NiagaraRibbonRendererProperties]}
  {emitterName=Propeller, rendererClasses [&NiagaraMeshRendererProperties]}
  {emitterName=Speed_Lines, rendererClasses [&NiagaraSpriteRendererProperties]}
  {emitterName=Tech_Shapes, rendererClasses [&NiagaraSpriteRendererProperties]}
 ]
 userVariables [
  User.bIsFrontEnd {type=NiagaraBool}
  User.bIsFrontEndPreview {type=NiagaraBool}
  User.ForwardDot {type=NiagaraFloat, defaultValue=0, description=""}
  User.RibbonIDOffset {type=NiagaraInt32}
  User.RightDot {type=NiagaraFloat, defaultValue=0, description=""}
  User.Variant {type=NiagaraFloat, defaultValue=0, description=""}
 ]
}
