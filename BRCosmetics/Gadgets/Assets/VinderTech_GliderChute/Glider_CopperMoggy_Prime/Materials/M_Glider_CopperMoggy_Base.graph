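# Base material for the CopperMoggy glider, built between the shared 2023 character parent PreFX and PostFX functions.
# An optional DigitalMosaic stage runs the CopperMoggy mosaic function, then mask-blends in a gradient-texture tint on base color and emissive.
# An optional Flame stage adds masked base color to emissive, applies the CopperMoggy flame overrides and drives temporal responsiveness.
# Named reroutes share masks, bounding-box UVs and two-layer scrolling Perlin noise.
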Material /BRCosmetics/Gadgets/Assets/VinderTech_GliderChute/Glider_CopperMoggy_Prime/Materials/M_Glider_CopperMoggy_Base
graph {
 properties {
  bAutomaticallySetUsageInEditor=false
  bUseMaterialAttributes=true
  bUsedWithMeshDeformer=true
  bUsedWithMorphTargets=true
  bUsedWithSkeletalMesh=true
  pixelDepthOffsetMode=PDOM_Legacy
  shadingModel=MSM_FromMaterialExpression
  usedShadingModels=Default Lit | Subsurface | Clear Coat | Subsurface Profile | Cloth
 }
 nodes [
  n1 = MaterialFunctionCall() {materialFunction=&MF_CharacterParent_2023_PreFX}
  n2 = MaterialFunctionCall!(Material=@n1.Result, Tech Art Mask=NamedRerouteUsage()) {materialFunction=/BRCosmetics/Characters/Player/Female/Medium/Bodies/F_MED_CopperMoggy_Swan/Skins/Prime/Materials/MF_CopperMoggy_DigitalMosaic}
  n3 = GetMaterialAttributes(None=@n2.Material Result) {attributeGetTypes [#1]}
  n4 = TextureSampleParameter2D(UVs=NamedRerouteUsage()) {
   group=CopperMoggy
   parameterName=Gradient
   samplerSource=SSM_Wrap_WorldGroupSettings
   texture=/BRCosmetics/Weapons/FORT_Melee/Pickaxe_CopperMoggy_Swan/Skins/Prime/Textures/T_CopperMoggy_Prime_ColorGradient_T_CopperMoggy_Prime_ColorGradient
  }
  n5 = Multiply(A=@n4.RGB, B=ScalarParameter() {defaultValue=1, group=CopperMoggy, parameterName=SoftEmissiveIntensity})
  n6 = NamedRerouteUsage()
  n7 = LinearInterpolate(
   A=@n3.Emissive Color
   B=Multiply(A=@n5, B=OneMinus(None=@n6))
   Alpha=ScalarParameter() {defaultValue=1, group=CopperMoggy, parameterName=SoftEmissiveAmount}
  )
  n8 = StaticSwitchParameter(
   True=BlendMaterialAttributes(
    A=@n2.Material Result
    B=SetMaterialAttributes!(MaterialAttributes=@n3.MaterialAttributes, Emissive Color=@n7, Base Color=@n7) {attributeSetTypes [#1, #2]}
    Alpha=NamedRerouteUsage()
   )
   False=@n1.Result
  ) {group=DigitalMosaic, parameterName=EnableDigitalMosaic}
  n9 = GetMaterialAttributes(None=@n8) {attributeGetTypes [#1, #2]}
  n10 = Multiply(
   A=Multiply(A=@n9.Base Color, B=ScalarParameter() {group=CopperMoggy|Flame, parameterName=BaseColorToEmissiveIntensity})
   B=ComponentMask(None=NamedRerouteUsage()) {b=true}
  )
  n11 = MaterialFunctionCall!(
   Material=SetMaterialAttributes!(MaterialAttributes=@n9.MaterialAttributes, Emissive Color=Add(A=@n9.Emissive Color, B=@n10)) {attributeSetTypes [#1]}
   TechArtMask=NamedRerouteUsage()
  ) {materialFunction=/BRCosmetics/Characters/Player/Female/Medium/Bodies/F_MED_CopperMoggy_Swan/Skins/Flame/Materials/MF_CopperMoggy_FlameOverrides}
  MaterialFunctionCall(
   Material=StaticSwitchParameter(True=@n11.Result, False=@n8) {group=Flame, parameterName=EnableFlameOverrides}
   M=NamedRerouteUsage()
   Pre Skinned Local Position=NamedRerouteUsage()
   Distance Blend=NamedRerouteUsage()
   Fresnel=NamedRerouteUsage()
  ) {materialFunction=&MF_CharacterParent_2023_PostFX}
  Multiply(A=Desaturation(None=@n3.Emissive Color, Fraction=Constant() {r=1}) {luminanceFactors=<rgba 0.0722 0.7152 0.2126 1>}, B=@n5)
  n12 = TextureSampleParameter2D() {
   group=CopperMoggy
   parameterName=TechArtMask_B
   samplerSource=SSM_Wrap_WorldGroupSettings
   samplerType=SAMPLERTYPE_Masks
   sortPriority=0
   texture=&T_WhiteMask
  }
  NamedRerouteDeclaration(None=@n12.B) {name=EmissiveMask, nodeColor=<rgba 1 0 0.7266 1>}
  NamedRerouteDeclaration(None=@n12.RGB) {name=AdditionalTechArtMask, nodeColor=<rgba 0 1 0.6484 1>}
  NamedRerouteDeclaration(None=@n1.M) {name=MaskTexture, nodeColor=<rgba 0 1 0.9531 1>}
  n13 = MaterialFunctionCall() {materialFunction=&BoundingBoxBased_0-1_UVW}
  NamedRerouteDeclaration(None=AppendVector(A=@n13.R, B=@n13.R)) {name=TopDownUVs, nodeColor=<rgba 0 1 0.3672 1>}
  n14 = VectorParameter() {group=CopperMoggy|Prime, parameterName=SoftNoiseSpeed_A, defaultValue=<rgba 0.1 0.15 0 0>}
  n15 = Add(
   A=Multiply(A=NamedRerouteUsage(), B=ScalarParameter() {defaultValue=0.333, group=CopperMoggy|Prime, parameterName=SoftNoiseTiling_A})
   B=Multiply(A=AppendVector(A=@n14.R, B=@n14.R), B=Time())
  )
  n16 = TextureObjectParameter() {group=CopperMoggy|Prime, parameterName=SoftNoise, texture=/Game/Global/Textures/Default/Noise/3xPerlin_M}
  n17 = VectorParameter() {group=CopperMoggy|Prime, parameterName=SoftNoiseSpeed_B, defaultValue=<rgba -0.1 0.1 0 0>}
  n18 = Add(
   A=Multiply(A=NamedRerouteUsage(), B=ScalarParameter() {defaultValue=0.2, group=CopperMoggy|Prime, parameterName=SoftNoiseTiling_B})
   B=Multiply(A=AppendVector(A=@n17.R, B=@n17.R), B=Time())
  )
  n19 = Multiply(
   A=Multiply(
    A=TextureSample(UVs=@n15, Tex=@n16) {samplerSource=SSM_Wrap_WorldGroupSettings, texture=&3xPerlin_M}.R
    B=TextureSample(UVs=@n18, Tex=@n16) {samplerSource=SSM_Wrap_WorldGroupSettings, texture=&3xPerlin_M}.G
   )
   B=ScalarParameter() {defaultValue=2, group=CopperMoggy|Prime, parameterName=SoftNoiseScale}
  ) {constB=2}
  n20 = ScalarParameter() {defaultValue=1, group=CopperMoggy|Prime, parameterName=SoftNoisePower}
  NamedRerouteDeclaration(None=Saturate(None=Power(Base=@n19, Exp=@n20))) {name=AnimatedSoftNoise, nodeColor=<rgba 0.1484 1 0 1>}
  NamedRerouteDeclaration(None=@n1.Pre skinned Local Position) {name=PreSkinnedLocalPos, nodeColor=<rgba 1 0.6328 0 1>}
  NamedRerouteDeclaration(None=@n1.Distance Blend) {name=DistanceBlend, nodeColor=<rgba 0.3125 0 1 1>}
  NamedRerouteDeclaration(None=@n1.Fresnel) {name=Fresnel, nodeColor=<rgba 0.0078 1 0 1>}
  TemporalResponsivenessOutput!(None=StaticSwitchParameter(True=@n11.Temporal Mask, False=Constant()) {group=Flame, parameterName=EnableFlameOverrides})
 ]
}
